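# Anim Blueprint for the Cacti skin's character mesh, copying the parent pose and driving four shoulder spike bones.
# Each spike bone is constrained with weights split between its side's clavicle and upper-arm twist bone.
# In first person, a 0.2-second blend switches to a layered pose that overlays spine and clavicle branches copied using baseSkelMesh.
# The event graph caches that mesh when customization loads and updates the first-person flag from the owning player AnimInstance.
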AnimBlueprint /BRCosmetics/Characters/Player/Female/Medium/Bodies/F_Med_Soldier_01/Skins/BR_Cacti/Meshes/F_MED_Cacti_AnimBp
target_skeleton=&/F_MED_Cacti_Skeleton
graphs [
 AnimGraph {nodes [
   Output Pose(Result=Blend Poses by bool(
     BlendPose_0=Use cached pose 'FP_targeting_conbine'()
     BlendPose_1=Use cached pose 'baseAnimGrp'()
     BlendTime_0=0.2
     bActiveValue=Get isFPScopeAnim() -> (isFPScopeAnim:bool)
    ) {blendType=ExpIn}) {layerGroup=None}
   n1 = "Constraint - const_spike_01 (PP)"!(
    ComponentPose=Local To Component(LocalPose=Copy Pose From Mesh() {bUseAttachedParent=true})
    Alpha=1
    ConstraintWeights_0=0.5775
    ConstraintWeights_1:float=0.4225
   ) -> (Pose) {
    alphaBoolBlend {blendOption=Linear}
    boneToModify {boneName=const_spike_01}
    constraintSetup [{targetBone {boneName=clavicle_r}}, {targetBone {boneName=upperarm_twist_01_r}}]
    constraintWeights [1,1]
   }
   n2 = "Constraint - const_spike_02 (PP)"!(ComponentPose=@n1, Alpha=1, ConstraintWeights_0=0.5548, ConstraintWeights_1:float=0.4452) -> (Pose) {
    alphaBoolBlend {blendOption=Linear}
    boneToModify {boneName=const_spike_02}
    constraintSetup [{targetBone {boneName=clavicle_r}}, {targetBone {boneName=upperarm_twist_01_r}}]
    constraintWeights [1,1]
   }
   n3 = "Constraint - const_spike_03 (PP)"!(ComponentPose=@n2, Alpha=1, ConstraintWeights_0=0.4689, ConstraintWeights_1:float=0.5311) -> (Pose) {
    alphaBoolBlend {blendOption=Linear}
    boneToModify {boneName=const_spike_03}
    constraintSetup [{targetBone {boneName=clavicle_r}}, {targetBone {boneName=upperarm_twist_01_r}}]
    constraintWeights [1,1]
   }
   n4 = "Constraint - const_spike_04 (PP)"!(ComponentPose=@n3, Alpha=1, ConstraintWeights_0=0.4694, ConstraintWeights_1:float=0.5306) -> (Pose) {
    alphaBoolBlend {blendOption=Linear}
    boneToModify {boneName=const_spike_04}
    constraintSetup [{targetBone {boneName=clavicle_l}}, {targetBone {boneName=upperarm_twist_01_l}}]
    constraintWeights [1,1]
   }
   Save cached pose 'baseAnimGrp'(Pose=Component To Local(ComponentPose=@n4)) {cacheName=baseAnimGrp}
   n5 = Layered blend per bone(
    BasePose=Use cached pose 'baseAnimGrp'()
    BlendPoses_0=Copy Pose From Mesh(SourceMeshComponent=Get baseSkelMesh() -> (baseSkelMesh:Skeletal Mesh Component Budgeted Object Reference)) {bUseAttachedParent=true}
   ) {layerSetup [{branchFilters [
       {boneName=clavicle_l}
       {boneName=clavicle_r}
       {boneName=spine_01}
       {boneName=spine_02}
       {boneName=spine_03}
       {boneName=spine_04}
       {boneName=spine_05}
      ]}]}
   Save cached pose 'FP_targeting_conbine'(Pose=@n5) {cacheName=FP_targeting_conbine}
  ]}
 EventGraph {nodes [
   Event BlueprintUpdateAnimation()
   > Set isFPScopeAnim(isFPScopeAnim:bool=Get bIsInFirstPerson!(self:Fort Player Anim Instance Object Reference=GetOwnerFortPlayerAnimInstance!(self:Custom Character Part Anim Instance Object Reference) -> (ReturnValue:Fort Player Anim Instance Object Reference)) -> (bIsInFirstPerson:bool)) -> (Output_Get:bool)
   n1 = Event OnCharacterCustomizationLoaded()
   > Set baseSkelMesh(baseSkelMesh:Skeletal Mesh Component Budgeted Object Reference=Get Array Item with EFort Custom Part Type(Array=@n1.CustomMeshParts)) -> (Output_Get:Skeletal Mesh Component Budgeted Object Reference)
  ]}
]
variables [
 As Fort First Person Camera Controller {type=FortFirstPersonCameraController, default=null}
 As Fort Weapon Component Scope Animation {type=FortWeaponComponent_ScopeAnimation, default=null}
 baseSkelMesh {type=SkeletalMeshComponentBudgeted}
 Clavicle_l_jnt_LocVector {type=Vector}
 Clavicle_l_jnt_RotVector {type=Rotator}
 Clavicle_r_jnt_LocVector {type=Vector}
 Clavicle_r_jnt_RotVector {type=Rotator}
 isFPScopeAnim {type=boolean}
 Spine_03_jnt_RotVector {type=Rotator}
 Spine_04_jnt_RotVector {type=Rotator}
 Spine_05_jnt_RotVector {type=Rotator}
]
